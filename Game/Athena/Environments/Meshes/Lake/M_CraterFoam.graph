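# Translucent lit foam material for the crater lake, usable on mesh particles, that builds sprites through the Sprite function's world position offset.
# Near-camera fade drives sprite scale and opacity, while world-aligned noise adds local-space displacement weighted by vertex red.
# A soft smoke shape blended with panning swirl noise drives color, opacity and dithered pixel depth offset.
# Switches select Dynamic Parameter scaling, particle alpha and world masking; higher quality levels add depth fade.
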Material /Game/Athena/Environments/Meshes/Lake/M_CraterFoam
graph {
 properties {
  bUsedWithMeshParticles=true
  bUsedWithNanite=true
  bUsedWithNiagaraMeshParticles=true
  blendMode=BLEND_Translucent
  pixelDepthOffsetMode=PDOM_Legacy
  translucencyLightingMode=TLM_VolumetricPerVertexNonDirectional
  usedShadingModels=Default Lit
 }
 nodes [
  n1 = DynamicParameter() {paramNames [SpriteScaleNea, SpriteScaleFar, RandomScale, OpacityMaskScale]}
  n2 = LinearInterpolate(B=@n1.SpriteScaleNea, Alpha=@n1.SpriteScaleNea) {desc=RandomScale}
  n3 = LinearInterpolate(B=@n1.SpriteScaleFar, Alpha=@n1.SpriteScaleFar) {desc=RandomScale}
  n4 = MaterialFunctionCall(Distance=ScalarParameter() {defaultValue=2500, parameterName=SoftenDistance}) {materialFunction=&Near_Camera_Fade}
  n5 = LinearInterpolate(
   A=StaticSwitchParameter(True=LinearInterpolate(Alpha=@n2) {constB=300}, False=ScalarParameter() {defaultValue=300, parameterName=SpriteScaleNear}) {parameterName=UseDPtoHandleSpriteScale}
   B=StaticSwitchParameter(True=LinearInterpolate(Alpha=@n3) {constB=300}, False=ScalarParameter() {defaultValue=300, parameterName=SpriteScaleFar}) {parameterName=UseDPtoHandleSpriteScale}
   Alpha=@n4.Result
  )
  n6 = TextureSample() {samplerType=SAMPLERTYPE_Normal, texture=&T_Sphere_N}
  n7 = TextureCoordinate() {uTiling=0.9, vTiling=1.8}
  n8 = Multiply(A=DynamicParameter() {paramNames [SpriteScaleNea, SpriteScaleFar, RandomScale, OpacityMaskScale]}.OpacityMaskScale, B=@n7)
  n9 = Multiply(A=@n7, B=ScalarParameter() {defaultValue=1, parameterName=ScaleOpacityMaskNoise})
  n10 = Panner(
   Coordinate=Add(A=StaticSwitchParameter(True=@n8, False=@n9) {parameterName=UseDynamicTextureSize}, B=VertexColor().R)
   Speed=VectorParameter() {parameterName=DropSpeed, defaultValue=<rgba 0 -2 0 1>}.RGB
  ) {speedY=-0.5}
  n11 = MaterialFunctionCall(
   Normal=TextureSample(UVs=@n10) {samplerType=SAMPLERTYPE_Normal, texture=/Game/Packages/DS_Fortnite_TRV_Terrain/Textures/T_TRV_SwampWater_01_N}.RGB
   Flatness=ScalarParameter() {defaultValue=0.7, parameterName=Flatten}
  ) {materialFunction=&FlattenNormal}
  n12 = VertexColor()
  n13 = TextureObject() {texture=/Game/Global/Textures/Default/Noise/TankMoteNoise}
  n14 = VectorParameter() {parameterName=Size, defaultValue=<rgba 400 400 400 0>}
  n15 = Time()
  n16 = VectorParameter() {parameterName=Offset, defaultValue=<rgba 0.005 0.005 0 0>}
  n17 = Multiply(A=WorldPosition().XYZ, B=Multiply(A=Multiply(A=@n15, B=@n16.RGB), B=ScalarParameter() {defaultValue=0.2, parameterName=Noisespeed}))
  n18 = LinearInterpolate(
   A=VectorParameter() {parameterName=OffsetPos, defaultValue=<rgba 0 0 -15 0>}.RGB
   B=VectorParameter() {parameterName=Offset_1, defaultValue=<rgba 0 0 60 0>}.RGB
   Alpha=ComponentMask(None=MaterialFunctionCall(TextureObject=@n13, TextureSize=@n14.RGB, WorldPosition=@n17) {materialFunction=&WorldAlignedTexture}.XYZ Texture) {g=true}
  )
  Add(
   A=MaterialFunctionCall!(
    Center Location=WorldPosition().XYZ
    0-1 UVs
    XY scale=AppendVector(A=@n5, B=@n5)
    OverridePivot
    "Normals (Optional)"=MaterialFunctionCall(BaseNormal=@n6.RGB, AdditionalNormal=@n11.Result) {materialFunction=&BlendAngleCorrectedNormals}.Result
   ) {materialFunction=&Sprite}.World position offset
   B=Transform(None=Multiply(A=@n12.R, B=@n18)) {transformSourceType=TRANSFORMSOURCE_Local}
  )
  n19 = MaterialFunctionCall() {materialFunction=/Game/Athena/Items/Gameplay/SilkyVeteran/FX/SlurpVatMask}
  n20 = TextureSample() {texture=/Game/Effects/Fort_Effects/Textures/T_Smoke_Soft}
  n21 = LinearInterpolate(
   A=@n20.G
   B=Multiply(
    A=@n20.G
    B=TextureSample(UVs=@n10) {samplerType=SAMPLERTYPE_LinearColor, texture=/Game/Weapons/WeaponSkins/Textures/T_Swirly_Noise_highContrast1}.G
   )
   Alpha=ScalarParameter() {defaultValue=0.5, parameterName=LErpoit}
  )
  n22 = Multiply(A=@n21, B=LinearInterpolate(A=ScalarParameter() {defaultValue=0.7, parameterName=OpacityMask}, Alpha=@n4.Result) {constA=0.5, constB=0.9})
  n23 = StaticSwitchParameter(True=Multiply(A=ParticleColor().A, B=@n22), False=@n22) {parameterName=UseParticleAlpha}
  n24 = Multiply(
   A=StaticSwitchParameter(True=Multiply(A=@n23, B=@n19.Result), False=@n23) {parameterName=Mask to World SphereMask}
   B=ScalarParameter() {defaultValue=0.6, parameterName=OpacityScale}
  )
  n25 = DepthFade(Opacity=@n24) {fadeDistanceDefault=40}
  MaterialFunctionCall!(
   Low _ PC Low and Med=@n24
   "High _ XB1, PS4"=@n25
   "Epic _ PC High, XSX, PS5"=@n25
   --------
   "Mobile Low _ PC Perf, Android"=@n24
   Mobile Med _ Switch=@n24
  ) {materialFunction=/Game/Global/MaterialFunctions/TechArt/MF_Fortnite_QualitySwitch_Scalar}
  n26 = TextureCoordinate() {uTiling=0.5}
  n27 = VectorParameter() {parameterName=DropSpeed_1, defaultValue=<rgba 0 -1 0 1>}
  Multiply(
   A=ScalarParameter() {defaultValue=0.01, parameterName=Roughness}
   B=TextureSample(UVs=Panner(Coordinate=@n26, Speed=@n27.RGB) {speedX=1, speedY=1}) {samplerType=SAMPLERTYPE_LinearColor, texture=/Game/Global/Textures/Default/Masks_and_Gradients/T_Scratch_Mask}.G
  )
  Multiply(
   A=LinearInterpolate(
    A=VectorParameter() {parameterName=ColorDim, defaultValue=<rgba 0.1904 0.5625 0.5625 1>}.RGB
    B=VectorParameter() {parameterName=ColorBright, defaultValue=<rgba 0.7782 0.9126 0.9167 1>}.RGB
    Alpha=@n21
   )
   B=ScalarParameter() {defaultValue=1.5, parameterName=ColorScale}
  )
  n28 = Constant()
  n29 = ScalarParameter() {defaultValue=110, parameterName=Offset}
  n30 = ShadingPathSwitch(
   Default=Multiply(
    A=Multiply(A=@n21, B=@n29)
    B=MaterialFunctionCall() {materialFunction=&DitherTemporalAA}.Result
   )
   Forward=@n28
  )
  QualitySwitch(Default=@n28, High=@n30, Medium=@n30)
 ]
}
